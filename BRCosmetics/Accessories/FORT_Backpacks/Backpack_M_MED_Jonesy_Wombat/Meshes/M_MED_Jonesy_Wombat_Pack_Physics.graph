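# Physics asset for the Jonesy Wombat backpack cosmetic, giving its dangling straps, shoelace, cloth, armor and weapon bones simulation bodies and joint limits.
# The pack body combines two capsules and a box and anchors every dyn_ chain, whose bones use one capsule or box each.
# Constraints link each chain bone to its parent with small angular limits.
# The shoelace root and whole clothA chain lock swing.
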PhysicsAsset /BRCosmetics/Accessories/FORT_Backpacks/Backpack_M_MED_Jonesy_Wombat/Meshes/M_MED_Jonesy_Wombat_Pack_Physics
bodies {
 dyn_armor_00 [{shapeType=Capsule, length=0.505, radius=0.505, rotation=<rot 0 -90 0>}]
 dyn_armor_01 [{shapeType=Box, extentX=10.9243, extentY=7.5, extentZ=15.0083, center=<0.0001 -2.5457 -3.1272>, rotation=<rot -4.0611 2.3873 -30.4848>}]
 dyn_clothA_00 [{shapeType=Box, extentX=3.0382, extentY=0.5, extentZ=1.3226}]
 dyn_clothA_01 [{shapeType=Box, extentX=3.0382, extentY=0.5, extentZ=4.8818, center=<0 0 -0.4462>, rotation=<rot 0 0 5.7038>}]
 dyn_clothA_02 [{shapeType=Box, extentX=3.0382, extentY=0.5, extentZ=4.8818, center=<0 -0.0744 -0.4462>}]
 dyn_clothA_03 [{shapeType=Box, extentX=3.0382, extentY=0.5, extentZ=4.8818, center=<0 -0.0744 -0.4462>}]
 dyn_cloth_00 [{shapeType=Box, extentX=2.2445, extentY=2.0192, extentZ=0.5, center=<0.2119 -0.1824 0.0742>}]
 dyn_cloth_01 [{shapeType=Box, extentX=5.6213, extentY=2.0192, extentZ=0.5, center=<3.2571 -0.3233 0.0742>, rotation=<rot 0 -2.2589 0>}]
 dyn_shoelace_00 [{shapeType=Capsule, length=3.2347, radius=1.322, rotation=<rot 0 -90 0>}]
 dyn_shoelace_01 [{shapeType=Capsule, length=5.5051, radius=1.2298, center=<0 0 -1.8349>, rotation=<rot 0 -90 0>}]
 dyn_shoelace_02 [{shapeType=Capsule, length=14.4087, radius=3.6931, center=<0.2571 -0.1399 -9.8731>, rotation=<rot -6.0313 -90 0.2248>}]
 dyn_strapA_00 [{shapeType=Capsule, length=1.2442, radius=1.166, rotation=<rot 0 -90 0>}]
 dyn_strapA_01 [{shapeType=Capsule, length=3.8201, radius=1.0734, center=<0 -0.6589 -2.8479>, rotation=<rot 0 0 9.0948>}]
 dyn_strapB_00 [{shapeType=Capsule, length=0.7995, radius=1.25, rotation=<rot 0 -90 0>}]
 dyn_strapB_01 [{shapeType=Capsule, length=1.7876, radius=1.25, center=<-0.0001 0 -2.6801>, rotation=<rot 0 -90 0>}]
 dyn_strapC_00 [{shapeType=Capsule, length=0.6108, radius=0.8737, rotation=<rot 0 -90 0>}]
 dyn_strapC_01 [{shapeType=Box, extentX=3.0843, extentY=0.5199, extentZ=6.4552, center=<0 0.1906 -1.8673>}]
 dyn_strapC_02 [{shapeType=Box, extentX=3.0843, extentY=0.5199, extentZ=6.4552, center=<0 0.1906 -1.8673>}]
 dyn_strapC_03 [{shapeType=Box, extentX=3.0843, extentY=0.5199, extentZ=6.4552, center=<0 0.1906 -1.8673>}]
 dyn_strapD_00 [{shapeType=Capsule, length=0.4918, radius=0.9111, center=<0 0 0.2491>, rotation=<rot 0 -90 0>}]
 dyn_strapD_01 [{shapeType=Box, extentX=3.7037, extentY=0.5, extentZ=4.7567, center=<0 0 -1.0656>}]
 dyn_strapD_02 [{shapeType=Box, extentX=3.7037, extentY=0.5, extentZ=4.7567, center=<0 0 -1.0656>}]
 dyn_weapon_00 [{shapeType=Capsule, length=3.5, radius=1.3187, rotation=<rot 0 -90 0>}]
 pack [
  {shapeType=Capsule, length=8.7637, radius=7.7411, center=<-0.4393 -6.9637 -9.2811>, rotation=<rot 1.2205 -94.8749 -89.9858>}
  {shapeType=Capsule, length=8.7637, radius=7.7411, center=<-0.5395 -6.6166 -5.7372>, rotation=<rot 1.2205 -94.8749 -89.9858>}
  {shapeType=Box, extentX=14.1408, extentY=3.5, extentZ=8.739, center=<-0.2196 -8.0505 2.6998>}
 ]
}
constraints [
 {bone1Name=dyn_strapA_00, bone2Name=pack, swing1LimitDegrees=5, swing2LimitDegrees=7, twistLimitDegrees=1}
 {bone1Name=dyn_strapA_01, bone2Name=dyn_strapA_00, swing1LimitDegrees=5, swing2LimitDegrees=7, twistLimitDegrees=10}
 {bone1Name=dyn_strapB_00, bone2Name=pack, swing1LimitDegrees=5, swing2LimitDegrees=5, twistLimitDegrees=1}
 {bone1Name=dyn_strapB_01, bone2Name=dyn_strapB_00, swing1LimitDegrees=5, swing2LimitDegrees=5, twistLimitDegrees=10}
 {bone1Name=dyn_strapC_00, bone2Name=pack, swing1LimitDegrees=2, swing2LimitDegrees=2, twistLimitDegrees=2}
 {bone1Name=dyn_strapC_01, bone2Name=dyn_strapC_00, swing1LimitDegrees=2, swing2LimitDegrees=2, twistLimitDegrees=2}
 {bone1Name=dyn_strapC_02, bone2Name=dyn_strapC_01, swing1LimitDegrees=2, swing2LimitDegrees=2, twistLimitDegrees=2}
 {bone1Name=dyn_strapC_03, bone2Name=dyn_strapC_02, swing1LimitDegrees=2, swing2LimitDegrees=2, twistLimitDegrees=2}
 {bone1Name=dyn_strapD_00, bone2Name=pack, swing1LimitDegrees=1, swing2LimitDegrees=1, twistLimitDegrees=1}
 {bone1Name=dyn_strapD_01, bone2Name=dyn_strapD_00, swing1LimitDegrees=1, swing2LimitDegrees=1, twistLimitDegrees=1}
 {bone1Name=dyn_strapD_02, bone2Name=dyn_strapD_01, swing1LimitDegrees=1, swing2LimitDegrees=1, twistLimitDegrees=1}
 {
  bone1Name=dyn_shoelace_00
  bone2Name=pack
  swing1LimitDegrees=0.1
  swing1Motion=Locked
  swing2LimitDegrees=0.1
  swing2Motion=Locked
  twistLimitDegrees=0.1
 }
 {bone1Name=dyn_shoelace_01, bone2Name=dyn_shoelace_00, swing1LimitDegrees=5, swing2LimitDegrees=5, twistLimitDegrees=10}
 {bone1Name=dyn_shoelace_02, bone2Name=dyn_shoelace_01, swing1LimitDegrees=5, swing2LimitDegrees=5, twistLimitDegrees=10}
 {bone1Name=dyn_armor_00, bone2Name=pack, swing1LimitDegrees=0, swing2LimitDegrees=0, twistLimitDegrees=5}
 {bone1Name=dyn_armor_01, bone2Name=dyn_armor_00, swing1LimitDegrees=0, swing2LimitDegrees=0, twistLimitDegrees=10}
 {bone1Name=dyn_weapon_00, bone2Name=pack, swing1LimitDegrees=8, swing2LimitDegrees=8, twistLimitDegrees=3}
 {bone1Name=dyn_cloth_00, bone2Name=pack, swing1LimitDegrees=10, swing2LimitDegrees=10, twistLimitDegrees=5}
 {bone1Name=dyn_cloth_01, bone2Name=dyn_cloth_00, swing1LimitDegrees=10, swing2LimitDegrees=10, twistLimitDegrees=10}
 {
  bone1Name=dyn_clothA_00
  bone2Name=pack
  swing1LimitDegrees=0
  swing1Motion=Locked
  swing2LimitDegrees=0
  swing2Motion=Locked
  twistLimitDegrees=5
 }
 {
  bone1Name=dyn_clothA_01
  bone2Name=dyn_clothA_00
  swing1LimitDegrees=0
  swing1Motion=Locked
  swing2LimitDegrees=0
  swing2Motion=Locked
  twistLimitDegrees=5
 }
 {
  bone1Name=dyn_clothA_02
  bone2Name=dyn_clothA_01
  swing1LimitDegrees=0
  swing1Motion=Locked
  swing2LimitDegrees=0
  swing2Motion=Locked
  twistLimitDegrees=5
 }
 {
  bone1Name=dyn_clothA_03
  bone2Name=dyn_clothA_02
  swing1LimitDegrees=0
  swing1Motion=Locked
  swing2LimitDegrees=0
  swing2Motion=Locked
  twistLimitDegrees=5
 }
]
