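# Animation Blueprint for the Jonesy Wombat backpack cosmetic, making the pack follow the body of its attach parent with simulated secondary motion.
# Each update, it reads world-space transforms of the parent's root, pelvis, spine, right clavicle and both upper arms.
# The AnimGraph replaces matching bones with those transforms, adds a small pack offset, then runs RigidBody physics relative to the pack bone with scaled component motion and overridden gravity.
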AnimBlueprint /BRCosmetics/Accessories/FORT_Backpacks/Backpack_M_MED_Jonesy_Wombat/Meshes/M_MED_Jonesy_Wombat_Pack_AnimBP
target_skeleton=&/M_MED_Jonesy_Wombat_Pack_Skeleton
graphs [
 AnimGraph {nodes [
   n1 = "Transform (Modify) Bone - Bone: body_root"(
    ComponentPose=Local To Component(LocalPose=Local Space Ref Pose())
    Alpha=1
    Translation=Get bodyRoot_tran() -> (bodyRoot_tran:Vector)
    Rotation=Get bodyRoot_rot() -> (bodyRoot_rot:Rotator)
    Scale="1,1,1"
   ) {
    rotationMode=BMM_Replace
    rotationSpace=BCS_WorldSpace
    scaleSpace=BCS_BoneSpace
    translationMode=BMM_Replace
    translationSpace=BCS_WorldSpace
    alphaBoolBlend {blendOption=Linear}
    alphaScaleBiasClamp {bInterpResult=true, inRange {max=0, min=-40}, outRange {max=0, min=1}}
    boneToModify {boneName=body_root}
   }
   n2 = "Transform (Modify) Bone - Bone: pelvis"(
    ComponentPose=@n1
    Alpha=1
    Translation=Get pelvis_tran() -> (pelvis_tran:Vector)
    Rotation=Get pelvis_rot() -> (pelvis_rot:Rotator)
    Scale="1,1,1"
   ) {
    rotationMode=BMM_Replace
    rotationSpace=BCS_WorldSpace
    scaleSpace=BCS_BoneSpace
    translationMode=BMM_Replace
    translationSpace=BCS_WorldSpace
    alphaBoolBlend {blendOption=Linear}
    alphaScaleBiasClamp {inRange {max=0, min=-40}, outRange {max=0, min=1}}
    boneToModify {boneName=pelvis}
   }
   n3 = "Transform (Modify) Bone - Bone: spine_01"(
    ComponentPose=@n2
    Alpha=1
    Translation=Get spine_01_tran() -> (spine_01_tran:Vector)
    Rotation=Get spine_01_rot() -> (spine_01_rot:Rotator)
    Scale="1,1,1"
   ) {
    rotationMode=BMM_Replace
    rotationSpace=BCS_WorldSpace
    scaleSpace=BCS_BoneSpace
    translationMode=BMM_Replace
    translationSpace=BCS_WorldSpace
    alphaBoolBlend {blendOption=Linear}
    alphaScaleBiasClamp {inRange {max=0, min=-40}, outRange {max=0, min=1}}
    boneToModify {boneName=spine_01}
   }
   n4 = "Transform (Modify) Bone - Bone: spine_02"(
    ComponentPose=@n3
    Alpha=1
    Translation=Get spine_02_tran() -> (spine_02_tran:Vector)
    Rotation=Get spine_02_rot() -> (spine_02_rot:Rotator)
    Scale="1,1,1"
   ) {
    rotationMode=BMM_Replace
    rotationSpace=BCS_WorldSpace
    scaleSpace=BCS_BoneSpace
    translationMode=BMM_Replace
    translationSpace=BCS_WorldSpace
    alphaBoolBlend {blendOption=Linear}
    alphaScaleBiasClamp {inRange {max=0, min=-40}, outRange {max=0, min=1}}
    boneToModify {boneName=spine_02}
   }
   n5 = "Transform (Modify) Bone - Bone: spine_03"(
    ComponentPose=@n4
    Alpha=1
    Translation=Get spine_03_tran() -> (spine_03_tran:Vector)
    Rotation=Get spine_03_rot() -> (spine_03_rot:Rotator)
    Scale="1,1,1"
   ) {
    rotationMode=BMM_Replace
    rotationSpace=BCS_WorldSpace
    scaleSpace=BCS_BoneSpace
    translationMode=BMM_Replace
    translationSpace=BCS_WorldSpace
    alphaBoolBlend {blendOption=Linear}
    alphaScaleBiasClamp {inRange {max=0, min=-40}, outRange {max=0, min=1}}
    boneToModify {boneName=spine_03}
   }
   n6 = "Transform (Modify) Bone - Bone: spine_04"(
    ComponentPose=@n5
    Alpha=1
    Translation=Get spine_04_tran() -> (spine_04_tran:Vector)
    Rotation=Get spine_04_rot() -> (spine_04_rot:Rotator)
    Scale="1,1,1"
   ) {
    rotationMode=BMM_Replace
    rotationSpace=BCS_WorldSpace
    scaleSpace=BCS_BoneSpace
    translationMode=BMM_Replace
    translationSpace=BCS_WorldSpace
    alphaBoolBlend {blendOption=Linear}
    alphaScaleBiasClamp {inRange {max=0, min=-40}, outRange {max=0, min=1}}
    boneToModify {boneName=spine_04}
   }
   n7 = "Transform (Modify) Bone - Bone: spine_05"(
    ComponentPose=@n6
    Alpha=1
    Translation=Get spine_05_tran() -> (spine_05_tran:Vector)
    Rotation=Get spine_05_rot() -> (spine_05_rot:Rotator)
    Scale="1,1,1"
   ) {
    rotationMode=BMM_Replace
    rotationSpace=BCS_WorldSpace
    scaleSpace=BCS_BoneSpace
    translationMode=BMM_Replace
    translationSpace=BCS_WorldSpace
    alphaBoolBlend {blendOption=Linear}
    alphaScaleBiasClamp {inRange {max=0, min=-40}, outRange {max=0, min=1}}
    boneToModify {boneName=spine_05}
   }
   n8 = "Transform (Modify) Bone - Bone: clavicle_r"(
    ComponentPose=@n7
    Alpha=1
    Translation=Get clavicle_r_tran() -> (clavicle_r_tran:Vector)
    Rotation=Get clavicel_r_rot() -> (clavicel_r_rot:Rotator)
    Scale="1,1,1"
   ) {
    rotationMode=BMM_Replace
    rotationSpace=BCS_WorldSpace
    scaleSpace=BCS_BoneSpace
    translationMode=BMM_Replace
    translationSpace=BCS_WorldSpace
    alphaBoolBlend {blendOption=Linear}
    alphaScaleBiasClamp {inRange {max=0, min=-40}, outRange {max=0, min=1}}
    boneToModify {boneName=clavicle_r}
   }
   n9 = "Transform (Modify) Bone - Bone: upperarm_r"(
    ComponentPose=@n8
    Alpha=1
    Translation=Get upperarm_r_tran() -> (upperarm_r_tran:Vector)
    Rotation=Get upperarm_r_rot() -> (upperarm_r_rot:Rotator)
    Scale="1,1,1"
   ) {
    rotationMode=BMM_Replace
    rotationSpace=BCS_WorldSpace
    scaleSpace=BCS_BoneSpace
    translationMode=BMM_Replace
    translationSpace=BCS_WorldSpace
    alphaBoolBlend {blendOption=Linear}
    alphaScaleBiasClamp {inRange {max=0, min=-40}, outRange {max=0, min=1}}
    boneToModify {boneName=upperarm_r}
   }
   n10 = "Transform (Modify) Bone - Bone: upperarm_l"(
    ComponentPose=@n9
    Alpha=1
    Translation=Get upperarm_l_tran() -> (upperarm_l_tran:Vector)
    Rotation=Get upperarm_l_rot() -> (upperarm_l_rot:Rotator)
    Scale="1,1,1"
   ) {
    rotationMode=BMM_Replace
    rotationSpace=BCS_WorldSpace
    scaleSpace=BCS_BoneSpace
    translationMode=BMM_Replace
    translationSpace=BCS_WorldSpace
    alphaBoolBlend {blendOption=Linear}
    alphaScaleBiasClamp {inRange {max=0, min=-40}, outRange {max=0, min=1}}
    boneToModify {boneName=upperarm_l}
   }
   n11 = "Transform (Modify) Bone - Bone: pack"(ComponentPose=@n10, Translation="0,-0.5,0", Rotation="0,0,6") {
    rotationMode=BMM_Additive
    rotationSpace=BCS_BoneSpace
    scaleMode=BMM_Additive
    scaleSpace=BCS_BoneSpace
    translationMode=BMM_Additive
    translationSpace=BCS_BoneSpace
    alphaBoolBlend {blendOption=Linear}
    boneToModify {boneName=pack}
    rotation {roll=6}
    translation {y=-0.5}
   }
   n12 = RigidBody(ComponentPose=@n11) {
    bClampLinearTranslationLimitToRefPose=true
    bOverrideWorldGravity=true
    overridePhysicsAsset=&/M_MED_Jonesy_Wombat_Pack_Physics
    simulationSpace=BaseBoneSpace
    alphaBoolBlend {blendOption=Linear}
    baseBoneRef {boneName=pack}
    componentAppliedLinearAccClamp=<1200 1200 1200>
    componentLinearAccScale=<0.35 0.45 0.45>
    componentLinearVelScale=<0.3 0.3 0.3>
    overrideWorldGravity {z=-980}
   }
   Output Pose(Result=Component To Local(ComponentPose=@n12)) {layerGroup=None}
  ]}
 EventGraph {nodes [
   Event BlueprintUpdateAnimation()
   n1 = GetPartSkeletalMeshComponent()
   > Is Valid(InputObject=@n1)
   n2 = GetAttachParent!(self:Scene Component Object Reference=@n1) -> (ReturnValue:Scene Component Object Reference)
   > Is Valid(InputObject=@n2)
   > n3 = Sequence!(execute) -> (
    then_0
    then_1
    then_2:Exec
    then_3:Exec
    then_4:Exec
    then_5:Exec
    then_6:Exec
    then_7:Exec
    then_8:Exec
    then_9:Exec
    then_10:Exec
    then_11:Exec
    then_12:Exec
    then_13:Exec
    then_14:Exec
    then_15:Exec
    then_16:Exec
    then_17:Exec
   )
   n4 = BreakTransform(InTransform=GetSocketTransform(self=@n2, InSocketName=spine_01, TransformSpace=RTS_World))
   Set spine_01_tran(execute=@n3.then_2, spine_01_tran:Vector=@n4) -> (Output_Get:Vector)
   > Set spine_01_rot(spine_01_rot:Rotator=@n4.Rotation) -> (Output_Get:Rotator)
   n5 = BreakTransform(InTransform=GetSocketTransform(self=@n2, InSocketName=pelvis, TransformSpace=RTS_World))
   Set pelvis_tran(execute=@n3.then_0, pelvis_tran:Vector=@n5) -> (Output_Get:Vector)
   > Set pelvis_rot(pelvis_rot:Rotator=@n5.Rotation) -> (Output_Get:Rotator)
   n6 = BreakTransform(InTransform=GetSocketTransform(self=@n2, InSocketName=root, TransformSpace=RTS_World))
   Set bodyRoot_tran(execute=@n3.then_1, bodyRoot_tran:Vector=@n6) -> (Output_Get:Vector)
   > Set bodyRoot_rot(bodyRoot_rot:Rotator=@n6.Rotation) -> (Output_Get:Rotator)
   n7 = BreakTransform(InTransform=GetSocketTransform(self=@n2, InSocketName=spine_02, TransformSpace=RTS_World))
   Set spine_02_tran(execute=@n3.then_3, spine_02_tran:Vector=@n7) -> (Output_Get:Vector)
   > Set spine_02_rot(spine_02_rot:Rotator=@n7.Rotation) -> (Output_Get:Rotator)
   n8 = BreakTransform(InTransform=GetSocketTransform(self=@n2, InSocketName=spine_03, TransformSpace=RTS_World))
   Set spine_03_tran(execute=@n3.then_4, spine_03_tran:Vector=@n8) -> (Output_Get:Vector)
   > Set spine_03_rot(spine_03_rot:Rotator=@n8.Rotation) -> (Output_Get:Rotator)
   n9 = BreakTransform(InTransform=GetSocketTransform(self=@n2, InSocketName=spine_04, TransformSpace=RTS_World))
   Set spine_04_tran(execute=@n3.then_5, spine_04_tran:Vector=@n9) -> (Output_Get:Vector)
   > Set spine_04_rot(spine_04_rot:Rotator=@n9.Rotation) -> (Output_Get:Rotator)
   n10 = BreakTransform(InTransform=GetSocketTransform(self=@n2, InSocketName=spine_05, TransformSpace=RTS_World))
   Set spine_05_tran(execute=@n3.then_6, spine_05_tran:Vector=@n10) -> (Output_Get:Vector)
   > Set spine_05_rot(spine_05_rot:Rotator=@n10.Rotation) -> (Output_Get:Rotator)
   n11 = BreakTransform(InTransform=GetSocketTransform(self=@n2, InSocketName=upperarm_r, TransformSpace=RTS_World))
   Set upperarm_r_tran(execute=@n3.then_7, upperarm_r_tran:Vector=@n11) -> (Output_Get:Vector)
   > Set upperarm_r_rot(upperarm_r_rot:Rotator=@n11.Rotation) -> (Output_Get:Rotator)
   n12 = BreakTransform(InTransform=GetSocketTransform(self=@n2, InSocketName=clavicle_r, TransformSpace=RTS_World))
   Set clavicle_r_tran(execute=@n3.then_8, clavicle_r_tran:Vector=@n12) -> (Output_Get:Vector)
   > Set clavicel_r_rot(clavicel_r_rot:Rotator=@n12.Rotation) -> (Output_Get:Rotator)
   n13 = BreakTransform(InTransform=GetSocketTransform(self=@n2, InSocketName=upperarm_l, TransformSpace=RTS_World))
   > Set upperarm_l_tran(upperarm_l_tran:Vector=@n13) -> (Output_Get:Vector)
   > Set upperarm_l_rot(upperarm_l_rot:Rotator=@n13.Rotation) -> (Output_Get:Rotator)
   TryGetPawnOwner()
  ]}
]
variables [
 bodyRoot_rot {type=Rotator}
 bodyRoot_tran {type=Vector}
 clavicel_r_rot {type=Rotator}
 clavicle_r_tran {type=Vector}
 pelvis_rot {type=Rotator}
 pelvis_tran {type=Vector}
 spine_01_rot {type=Rotator}
 spine_01_tran {type=Vector}
 spine_02_rot {type=Rotator}
 spine_02_tran {type=Vector}
 spine_03_rot {type=Rotator}
 spine_03_tran {type=Vector}
 spine_04_rot {type=Rotator}
 spine_04_tran {type=Vector}
 spine_05_rot {type=Rotator}
 spine_05_tran {type=Vector}
 upperarm_l_rot {type=Rotator}
 upperarm_l_tran {type=Vector}
 upperarm_r_rot {type=Rotator}
 upperarm_r_tran {type=Vector}
]
