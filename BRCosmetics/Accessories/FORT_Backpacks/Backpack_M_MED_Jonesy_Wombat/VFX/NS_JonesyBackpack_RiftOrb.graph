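# Niagara effect for the rift orb on the Jonesy Wombat backpack cosmetic, escalating through four stages selected by User.Stage.
# Layers a plasma ball mesh, rotating shard meshes, fire and arc sprites, curl-noise embers pulled inward, and a flickering light.
# Each emitter indexes per-stage arrays with User.Stage to shift blue toward red-orange and tune spawn rates, sizes, forces, flicker or material parameters.
# Light also sets per-stage enable flags alongside three component renderers.
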NiagaraSystem /BRCosmetics/Accessories/FORT_Backpacks/Backpack_M_MED_Jonesy_Wombat/VFX/NS_JonesyBackpack_RiftOrb
dependencies {
 usedDataInterfaces [
  &NiagaraDataInterfaceCurve
  &NiagaraDataInterfaceArrayFloat
  &NiagaraDataInterfaceColorCurve
  &NiagaraDataInterfaceArrayInt32
 ]
 usedDynamicInputs [&MakeVectorFromLinearColorRGB]
 usedModules [&SystemState]
}
emitters {
 Arcs {
  emitterUpdateScript [
   &EmitterState()
   &SpawnRate(
    Spawn Probability=0.5
    SpawnRate=&SelectFloatFromArray(Direct Array Index=$User.Stage, Float Selection Array=DataInterface() {propertyValues {FloatData [0,1,3,5]}})
    Use Spawn Probability=true
   )
  ]
  particleSpawnScript [
   /Niagara/Modules/Spawn/Initialization/InitializeParticle(
    Color=<rgba 100 100 100 1>
    Lifetime=&RandomRangeFloat(Minimum=0.1, Maximum=0.13)
    Module.Write Color=true
    Module.Write Lifetime=true
    Module.Write Mass=true
    Module.Write Material Random=true
    Module.Write Position=false
    Module.Write Scale=false
    Module.Write SpriteRotation=true
    Module.Write SpriteSize=true
    Position=&SimulationPosition()
    Sprite Rotation=&RandomRangeFloat(Minimum=-180, Maximum=180)
    Sprite Size=&Vector2DFromFloat(Value=&RandomRangeFloat(Minimum=6, Maximum=10))
   )
   &SubUVAnimation(End Frame=3, SubUV Animation Mode:ENiagara_SubUVLookupMode=::NewEnumerator2)
  ]
  particleUpdateScript [
   &ParticleState()
   &CameraOffset()
   &ScaleSpriteSize(Scale Factor=&Vector2DFromFloat(Value=&FloatFromCurve(
      FloatCurve=DataInterface() {propertyValues {Curve {keys [
          {arriveTangent=1.7865, interpMode=RCIM_Cubic, time=0, value=0.5}
          {arriveTangent=0.6344, interpMode=RCIM_Cubic, leaveTangent=0.6344, tangentMode=RCTM_User, time=1, value=1}
         ]}}}
      Scale Curve=&SelectFloatFromArray(Direct Array Index=$User.Stage, Float Selection Array=DataInterface() {propertyValues {FloatData [1,1,1,1.5]}})
     )))
   &Color(Color=&LerpLinearColors(
     StartColor=&ColorFromCurve(ColorCurve=DataInterface() {propertyValues {
        AlphaCurve {keys [<at 0.5586 1 RCIM_Cubic>,<at 1.0091 0 RCIM_Cubic>]}
        BlueCurve {keys [<at 0 1 RCIM_Cubic>,<at 1 1 RCIM_Cubic>]}
        GreenCurve {keys [<at 0 0.2242 RCIM_Cubic>,<at 1 0.1557 RCIM_Cubic>]}
        RedCurve {keys [<at 0 0.5656 RCIM_Cubic>,<at 1 0.1073 RCIM_Cubic>]}
       }})
     EndColor=&ColorFromCurve(ColorCurve=DataInterface() {propertyValues {
        AlphaCurve {keys [<at 0.5586 1 RCIM_Cubic>,<at 1.0091 0 RCIM_Cubic>]}
        BlueCurve {keys [<at 0 1.2408 RCIM_Cubic>,<at 0.982 0>]}
        GreenCurve {keys [<at 0 1.7901 RCIM_Cubic>,<at 0.982 0.2738>]}
        RedCurve {keys [<at 0 5 RCIM_Cubic>,<at 0.982 5>]}
       }})
     LerpFactor=&SelectFloatFromArray(Direct Array Index=$User.Stage, Float Selection Array=DataInterface() {propertyValues {FloatData [0,0,0,1]}})
    ))
   &SpriteRotationRate(Rotation Rate=500)
  ]
  rendererClasses [&NiagaraSpriteRendererProperties]
  renderers [{rendererClass=&NiagaraSpriteRendererProperties}]
 }
 Embers {
  emitterUpdateScript [
   &EmitterState()
   &SpawnRate(
    SpawnRate=&MakeFloatFromInt(Int=/Game/Effects/Niagara/DynamicInputs/Arrays/FN_SelectIntFromArray(
      Array Sampling Mode:ENiagaraArraySamplingMode=::NewEnumerator1
      Interpolation:NiagaraFloat=?
      Evaluation Type:ENiagaraRandomnessEvaluation=::NewEnumerator0
      Normalized Interpolation:NiagaraBool=true
      Direct Array Index:NiagaraInt32=$User.Stage
      Recalculate Random Each Loop:NiagaraBool=false
      Int Selection Array:NiagaraDataInterfaceArrayInt32=DataInterface() {propertyValues {IntData [40,40,30,20]}}
      Randomness Mode:ENiagaraRandomnessMode=::NewEnumerator0
      Random Seed:NiagaraInt32=?
      Fixed Random Seed:NiagaraBool=false
     ))
    Use Spawn Probability=false
   )
  ]
  particleSpawnScript [
   &InitializeParticle(
    Lifetime=3
    Module.Write Color=true
    Module.Write Lifetime=true
    Module.Write Mass=true
    Module.Write Material Random=true
    Module.Write Position=true
    Module.Write Scale=false
    Module.Write SpriteRotation=false
    Module.Write SpriteSize=true
    Position=&SimulationPosition()
    Sprite Size=&Vector2DFromFloat(Value=&Multiply_Float(
      A=&RandomRangeFloat(Minimum=0.125, Maximum=0.25)
      B=&SelectFloatFromArray(Direct Array Index=$User.Stage, Float Selection Array=DataInterface() {propertyValues {FloatData [1.5,1.5,2,3.3]}})
     ))
   )
   &SphereLocation(
    MaskSpawnGroup=false
    Override Seed=false
    Sphere Radius=5
    Surface Only=false
    UseNonUniformScale=false
    UseOffset=false
    UseSphereOrientation=false
   )
   /SetVariables(Particles.RandomAtSpawn:NiagaraFloat=&RandomRangeFloat(Minimum=0, Maximum=1)) {isSetParameters=true}
  ]
  particleUpdateScript [
   &ParticleState()
   &CameraOffset(Camera Offset Amount=5)
   /Niagara/Modules/Update/Forces/CurlNoiseForce(
    Noise Frequency=0.1
    Noise Strength=&SelectFloatFromArray(Direct Array Index=$User.Stage, Float Selection Array=DataInterface() {propertyValues {FloatData [4,8,16,32]}})
    Pan Noise=true
    Pan Noise Field=<0 0 1>
   )
   &PointAttractionForce(
    Attraction Radius=15
    Attraction Strength=&SelectFloatFromArray(
     Direct Array Index=$User.Stage
     Evaluation Type=::NewEnumerator1
     Float Selection Array=DataInterface() {propertyValues {FloatData [0.25,0.5,1,3]}}
    )
    Falloff Exponent=1
    Kill Radius=3
    Kill Within Radius=true
    Module.Use Falloff=true
    UseAttractorOffset=false
   )
   &SolveForcesAndVelocity(Clamp Velocity=false, Limit Acceleration=false)
   &KillParticlesInVolume(Invert Volume=true, Sphere Radius=5) {enabled=false}
   &ScaleColor(
    Scale Mode=::NewEnumerator1
    Scale RGBA=&MakeVector4_FromLinearColor(LinearColor=&LerpLinearColors(
      StartColor=&MultiplyLinearColor(
       Linear Color=<rgba 1.9607 6.3072 15 1>
       Scale Factor=&MakeVector4_FromFloat(Float=&FloatFromCurve(FloatCurve=DataInterface() {propertyValues {Curve {keys [<at 0 1 RCIM_Cubic>,<at 1 0>]}}}))
      )
      EndColor=&MultiplyLinearColor(
       Linear Color=<rgba 5 0.4471 0.3906 1>
       Scale Factor=&MakeVector4_FromFloat(Float=&FloatFromCurve(FloatCurve=DataInterface() {propertyValues {Curve {keys [<at 0 1 RCIM_Cubic>,<at 1 0>]}}}))
      )
      LerpFactor=&SelectFloatFromArray(Direct Array Index=$User.Stage, Float Selection Array=DataInterface() {propertyValues {FloatData [0,0,0.5,1]}})
     ))
    ScaleRGBA=true
   )
   &self:ClampMotes(
    Chance to Clamp Position:NiagaraFloat=&SelectFloatFromArray(Direct Array Index=$User.Stage, Float Selection Array=DataInterface() {propertyValues {FloatData [1,0.9,0.6,0.1]}})
    Sphere Radius=5.5
   ) {module=ClampMotes}
  ]
  rendererClasses [&NiagaraSpriteRendererProperties]
  renderers [{rendererClass=&NiagaraSpriteRendererProperties}]
 }
 Fire {
  emitterUpdateScript [
   &EmitterState()
   &SpawnRate(
    SpawnRate=&SelectFloatFromArray(Direct Array Index=$User.Stage, Float Selection Array=DataInterface() {propertyValues {FloatData [1,1,2,3]}})
    Use Spawn Probability=false
   )
  ]
  particleSpawnScript [
   &InitializeParticle(
    Color=&LerpLinearColors(
     StartColor=<rgba 0.6061 1.7045 5 0.5>
     EndColor=<rgba 30 0.6 0 0.5>
     LerpFactor=&SelectFloatFromArray(Direct Array Index=$User.Stage, Float Selection Array=DataInterface() {propertyValues {FloatData [0,0.05,0.1,1]}})
    )
    Lifetime=&SelectFloatFromArray(Direct Array Index=$User.Stage, Float Selection Array=DataInterface() {propertyValues {FloatData [3.6,2.6,2,1.3]}})
    Module.Write Color=true
    Module.Write Lifetime=true
    Module.Write Mass=true
    Module.Write Material Random=true
    Module.Write Position=false
    Module.Write Scale=false
    Module.Write SpriteRotation=true
    Module.Write SpriteSize=true
    Position=&SimulationPosition()
    Sprite Rotation=&RandomRangeFloat(Minimum=-180, Maximum=180)
    Sprite Size=<20 20>
   )
   /SetVariables(Particles.RandomFrame=&RandomRangeInt(Minimum=0, Maximum=63)) {isSetParameters=true}
  ]
  particleUpdateScript [
   &ParticleState()
   &CameraOffset(Camera Offset Amount=5)
   &FN_SubUVAnimation(End Frame=&Add_Integer(A=$Particles.RandomFrame, B=63), Start Frame=$Particles.RandomFrame, SubUV Loop Count=0.75)
   &ScaleSpriteSize(Scale Factor=&Vector2DFromFloat(Value=&FloatFromCurve(
      FloatCurve=DataInterface() {propertyValues {Curve {keys [
          {arriveTangent=2.5076, interpMode=RCIM_Cubic, leaveTangent=2.5076, tangentMode=RCTM_User, time=0.0023, value=0}
          {arriveTangent=0.0897, interpMode=RCIM_Cubic, leaveTangent=0.0897, tangentMode=RCTM_User, time=1, value=1}
         ]}}}
      Scale Curve=&SelectFloatFromArray(Direct Array Index=$User.Stage, Float Selection Array=DataInterface() {propertyValues {FloatData [1.2,1.25,1.3,1.4]}})
     )))
   &ScaleColor(
    Scale Alpha=&FloatFromCurve(FloatCurve=DataInterface() {propertyValues {Curve {keys [
         <at 0 0 RCIM_Cubic>
         {arriveTangent=5.1321, interpMode=RCIM_Cubic, leaveTangent=0.2, tangentMode=RCTM_Break, time=0.5, value=1}
         {arriveTangent=-1.4896, interpMode=RCIM_Cubic, leaveTangent=-1.4896, tangentMode=RCTM_User, time=1, value=0}
        ]}}})
    ScaleA=true
    ScaleRGB=true
   )
  ]
  rendererClasses [&NiagaraSpriteRendererProperties]
  renderers [{rendererClass=&NiagaraSpriteRendererProperties}]
 }
 Light {
  emitterUpdateScript [
   &EmitterState(
    Life Cycle Mode=::NewEnumerator1
    Loop Behavior=::NewEnumerator1
    Loop Duration Mode=::NewEnumerator1
    Reset Age On Awaken=true
    Scalability Mode=::NewEnumerator1
    UseLoopDelay=false
   )
   &SpawnBurst_Instantaneous(Spawn Count=1, Use Spawn Probability=false)
  ]
  particleSpawnScript [
   &InitializeParticle(
    Color=<rgba 1.3327 2.5551 5 0.05>
    Module.Write Color=true
    Module.Write Lifetime=true
    Module.Write Mass=true
    Module.Write Material Random=true
    Module.Write Position=false
    Module.Write Scale=false
    Module.Write SpriteRotation=false
    Module.Write SpriteSize=true
    Position=&SimulationPosition()
    Sprite Size=&Vector2DFromFloat(Value=30)
   )
   /SetVariables(Particles.Enabled01:NiagaraBool=&Boolean_OrOperation(
     Bool1=&SetBoolByIntComparison(A=$User.Stage, B=0, Comparison Type=::NewEnumerator4)
     Bool2=&SetBoolByIntComparison(A=$User.Stage, B=1, Comparison Type=::NewEnumerator4)
    )) {isSetParameters=true}
   /SetVariables(Particles.Enabled02:NiagaraBool=&SetBoolByIntComparison(A=$User.Stage, B=2, Comparison Type=::NewEnumerator4)) {isSetParameters=true}
   /SetVariables(Particles.Enabled03:NiagaraBool=&SetBoolByIntComparison(A=$User.Stage, B=3, Comparison Type=::NewEnumerator4)) {isSetParameters=true}
  ]
  particleUpdateScript [
   &ParticleState(
    Kill Particles When Lifetime Has Elapsed=false
    Let Infinitely Lived Particles Die When Emitter Deactivates=true
    Loop Particles Lifetime=true
   )
   &CameraOffset(Camera Offset Amount=8)
   &FN_Light_v3(
    Color=&LerpLinearColors(
     StartColor=<rgba 0.8485 2.3864 7 1>
     EndColor=<rgba 2 0.05 0 0.8>
     LerpFactor=&SelectFloatFromArray(Direct Array Index=$User.Stage, Float Selection Array=DataInterface() {propertyValues {FloatData [0,0.05,0.1,1]}})
    )
    Create Color Attribute=true
    Create Exponent Attribute=true
    Create Radius Attribute=true
    Exponent=5
    Radius:NiagaraFloat=&Add_Float(
     A=75
     B=&FloatFromCurve(
      FloatCurve=DataInterface() {propertyValues {Curve {keys [<at 0 0>,<at 0.0579 0.8067>,<at 0.1401 0.355>,<at 0.2024 0.5966>,<at 0.3068 0.0924>,<at 0.4135 0.7857>,<at 0.4601 0.2185>,<at 0.5446 0.5126>,<at 0.7001 0.0819>,<at 0.7713 0.8382>,<at 0.9024 0.25>,<at 1 0>]}}}
      Scale Curve=&SelectFloatFromArray(Direct Array Index=$User.Stage, Float Selection Array=DataInterface() {propertyValues {FloatData [0,2.5,5,10]}})
     )
    )
    Scale Color:NiagaraFloat=&Add_Float(
     A=1
     B=&FloatFromCurve(
      FloatCurve=DataInterface() {propertyValues {Curve {keys [<at 0 0>,<at 0.0579 0.8067>,<at 0.1401 0.355>,<at 0.2024 0.5966>,<at 0.3068 0.0924>,<at 0.4135 0.7857>,<at 0.4601 0.2185>,<at 0.5446 0.5126>,<at 0.7001 0.0819>,<at 0.7713 0.8382>,<at 0.9024 0.25>,<at 1 0>]}}}
      Scale Curve=&SelectFloatFromArray(Direct Array Index=$User.Stage, Float Selection Array=DataInterface() {propertyValues {FloatData [0,0.25,0.5,1]}})
     )
    )
   )
   &ScaleSpriteSize(Scale Factor=&Vector2DFromFloat(Value=&Add_Float(
      A=1
      B=&FloatFromCurve(
       FloatCurve=DataInterface() {propertyValues {Curve {keys [<at 0 0>,<at 0.0579 0.8067>,<at 0.1401 0.355>,<at 0.2024 0.5966>,<at 0.3068 0.0924>,<at 0.4135 0.7857>,<at 0.4601 0.2185>,<at 0.5446 0.5126>,<at 0.7001 0.0819>,<at 0.7713 0.8382>,<at 0.9024 0.25>,<at 1 0>]}}}
       Scale Curve=&SelectFloatFromArray(Direct Array Index=$User.Stage, Float Selection Array=DataInterface() {propertyValues {FloatData [0,0.1,0.2,0.35]}})
      )
     )))
   &ScaleColor(
    Scale Mode=::NewEnumerator1
    Scale RGBA=&MakeVector4_FromLinearColor(LinearColor=&LerpLinearColors(
      StartColor=<rgba 0.2 0.6833 1.65 0.7>
      EndColor=<rgba 1.6 0.05 0.25 1>
      LerpFactor=&SelectFloatFromArray(Direct Array Index=$User.Stage, Float Selection Array=DataInterface() {propertyValues {FloatData [0,0.05,0.1,1]}})
     ))
    ScaleRGBA=true
   )
  ]
  rendererClasses [
   &NiagaraSpriteRendererProperties
   &NiagaraLightRendererProperties
   &NiagaraComponentRendererProperties
  ]
  renderers [
   {rendererClass=&NiagaraSpriteRendererProperties}
   {rendererClass=&NiagaraLightRendererProperties, rendererIndex=1}
   {rendererClass=&NiagaraComponentRendererProperties, rendererIndex=2}
   {rendererClass=&NiagaraComponentRendererProperties, rendererIndex=3}
   {rendererClass=&NiagaraComponentRendererProperties, rendererIndex=4}
  ]
 }
 PlasmaBall001 {
  emitterUpdateScript [
   &EmitterState(
    Life Cycle Mode=::NewEnumerator1
    Loop Behavior=::NewEnumerator1
    Loop Duration Mode=::NewEnumerator1
    Reset Age On Awaken=true
    Scalability Mode=::NewEnumerator1
    UseLoopDelay=true
   )
   &SpawnBurst_Instantaneous(Spawn Count=1, Use Spawn Probability=false)
  ]
  particleSpawnScript [&InitializeParticle(
    Color=&LerpLinearColors(
     StartColor=<rgba 0.3212 1.101 5 1>
     EndColor=<rgba 6 0.6 0.4 1>
     LerpFactor=&SelectFloatFromArray(Direct Array Index=$User.Stage, Float Selection Array=DataInterface() {propertyValues {FloatData [0,0,0.9,1]}})
    )
    Mesh Scale=&VectorFromFloat(Value=0.1)
    Module.Write Color=true
    Module.Write Lifetime=false
    Module.Write Mass=true
    Module.Write Material Random=true
    Module.Write Position=true
    Module.Write Scale=true
    Module.Write SpriteRotation=false
    Module.Write SpriteSize=false
    Position=&SimulationPosition()
    Sprite Size=<25 25>
   )]
  particleUpdateScript [
   &DynamicMaterialParameters(
    Index 0 Param 3=&SelectFloatFromArray(Direct Array Index=$User.Stage, Float Selection Array=DataInterface() {propertyValues {FloatData [-0.15,-0.265,-0.4,-0.6]}})
    Module.Param0WriteEnabled=true
    Module.Param1WriteEnabled=true
    Module.Param2WriteEnabled=true
    Module.Param3WriteEnabled=true
   )
   &ParticleState(Kill Particles When Lifetime Has Elapsed=false, Let Infinitely Lived Particles Die When Emitter Deactivates=true)
   &UpdateMeshOrientation(Rotation Rate=0, Rotation Vector=<0 0 1>)
   &CameraOffset(Camera Offset Amount=8)
  ]
  rendererClasses [&NiagaraMeshRendererProperties]
  renderers [{rendererClass=&NiagaraMeshRendererProperties, rendererIndex=0}]
 }
 Shards {
  emitterUpdateScript [
   &EmitterState(
    Life Cycle Mode=::NewEnumerator1
    Loop Behavior=::NewEnumerator1
    Loop Duration Mode=::NewEnumerator1
    Reset Age On Awaken=true
    Scalability Mode=::NewEnumerator1
    UseLoopDelay=true
   )
   &SpawnBurst_Instantaneous(Spawn Count=2, Use Spawn Probability=false)
  ]
  particleSpawnScript [
   /SetVariables(Particles.MaterialRandom:NiagaraFloat=&RandomRangeFloat(Minimum=0, Maximum=1)) {isSetParameters=true, module=SetVariables_#1}
   &InitializeParticle(
    Color=&LerpLinearColors(
     StartColor=<rgba 1 1 1 1>
     EndColor=<rgba 5 0.4919 0.026 0.3>
     LerpFactor=&SelectFloatFromArray(Direct Array Index=$User.Stage, Float Selection Array=DataInterface() {propertyValues {FloatData [0,0,0,1]}})
    )
    Mesh Scale=&MakeVector(
     X=&RandomRangeFloat(Minimum=0.15, Maximum=0.18)
     Y=&RandomRangeFloat(Minimum=0.15, Maximum=0.18)
     Z=0
    )
    Module.Write Color=true
    Module.Write Lifetime=true
    Module.Write Mass=true
    Module.Write Material Random=true
    Module.Write Position=true
    Module.Write Scale=true
    Module.Write SpriteRotation=false
    Module.Write SpriteSize=false
    Sprite Size=<25 25>
   )
   &self:DivideParticles(HowManyRenderers=3) {module=DivideParticles}
   &InitialMeshOrientation(
    Orientation Axis=<0 1 0>
    Orientation Vector=<0 0 1>
    Rotation=&RandomRangeVector(Minimum=<-1 -1 -1>, Maximum=<1 1 1>)
    Use Orientation Vector=true
    Use Rotation Vector=true
   )
  ]
  particleUpdateScript [
   &ParticleState(Kill Particles When Lifetime Has Elapsed=false, Let Infinitely Lived Particles Die When Emitter Deactivates=true)
   &UpdateMeshOrientation(
    Rotation Rate=&Multiply_Float(
     A=&RandomRangeFloat(Minimum=0.13, Maximum=0.075)
     B=&MakeFloatFromInt(Int=&Add_Integer(A=1, B=$User.Stage))
    )
    Rotation Vector=<0 0 1>
   )
   &DynamicMaterialParameters(
    Index 0 Param 1=&SelectFloatFromArray(Direct Array Index=$User.Stage, Float Selection Array=DataInterface() {propertyValues {FloatData [0,0,2,7]}})
    Index 0 Param 2=&SelectFloatFromArray(Direct Array Index=$User.Stage, Float Selection Array=DataInterface() {propertyValues {FloatData [0,0,5,35]}})
    Index 0 Param 3=&SelectFloatFromArray(Direct Array Index=$User.Stage, Float Selection Array=DataInterface() {propertyValues {FloatData [0,0,0.35,1]}})
    Module.Param0WriteEnabled=true
    Module.Param1WriteEnabled=true
    Module.Param2WriteEnabled=true
    Module.Param3WriteEnabled=false
   )
  ]
  rendererClasses [&NiagaraMeshRendererProperties]
  renderers [
   {rendererClass=&NiagaraMeshRendererProperties, rendererIndex=0}
   {rendererClass=&NiagaraMeshRendererProperties, rendererIndex=1}
   {rendererClass=&NiagaraMeshRendererProperties, rendererIndex=2}
  ]
 }
 ShardsSmall {
  emitterUpdateScript [
   &EmitterState(
    Life Cycle Mode=::NewEnumerator1
    Loop Behavior=::NewEnumerator1
    Loop Duration Mode=::NewEnumerator1
    Reset Age On Awaken=true
    Scalability Mode=::NewEnumerator1
    UseLoopDelay=true
   )
   &SpawnBurst_Instantaneous(Spawn Count=7, Use Spawn Probability=false)
  ]
  particleSpawnScript [
   /SetVariables(Particles.MaterialRandom:NiagaraFloat=&RandomRangeFloat(Minimum=0, Maximum=1)) {isSetParameters=true, module=SetVariables_#1}
   &InitializeParticle(
    Color=&LerpLinearColors(
     StartColor=<rgba 1 1 1 1>
     EndColor=<rgba 5 0.3572 0.1128 1>
     LerpFactor=&SelectFloatFromArray(Direct Array Index=$User.Stage, Float Selection Array=DataInterface() {propertyValues {FloatData [0,0,0,1]}})
    )
    Mesh Scale=&VectorFromFloat(Value=&Lerp_Float(A=0.04, B=0.1, Alpha=$Particles.MaterialRandom))
    Module.Write Color=true
    Module.Write Lifetime=true
    Module.Write Mass=true
    Module.Write Material Random=true
    Module.Write Position=true
    Module.Write Scale=true
    Module.Write SpriteRotation=false
    Module.Write SpriteSize=false
    Sprite Size=<25 25>
   )
   &self:DivideParticles(HowManyRenderers=3) {module=DivideParticles}
   &InitialMeshOrientation(
    Orientation Axis=<0 1 0>
    Orientation Vector=<0 0 1>
    Rotation=&RandomRangeVector(Minimum=<-1 -1 -1>, Maximum=<1 1 1>)
    Use Orientation Vector=true
    Use Rotation Vector=true
   )
   &SphereLocation(
    MaskSpawnGroup=false
    Override Seed=false
    Sphere Radius=&Lerp_Float(A=0, B=2, Alpha=$Particles.MaterialRandom)
    Surface Only=false
    UseNonUniformScale=false
    UseOffset=false
    UseSphereOrientation=false
   )
  ]
  particleUpdateScript [
   &ParticleState(Kill Particles When Lifetime Has Elapsed=false, Let Infinitely Lived Particles Die When Emitter Deactivates=true)
   &UpdateMeshOrientation(
    Rotation Coordinate Space=::NewEnumerator3
    Rotation Rate=&Multiply_Float(
     A=&RandomRangeFloat(Minimum=0.13, Maximum=0.075)
     B=&MakeFloatFromInt(Int=&Add_Integer(A=1, B=$User.Stage))
    )
    Rotation Vector=<1 0 0>
   )
   &DynamicMaterialParameters(
    Index 0 Param 1=&SelectFloatFromArray(Direct Array Index=$User.Stage, Float Selection Array=DataInterface() {propertyValues {FloatData [0,0,2,5]}})
    Index 0 Param 2=&SelectFloatFromArray(Direct Array Index=$User.Stage, Float Selection Array=DataInterface() {propertyValues {FloatData [0,0,10,35]}})
    Index 0 Param 3=&SelectFloatFromArray(Direct Array Index=$User.Stage, Float Selection Array=DataInterface() {propertyValues {FloatData [0,0,0.6,1.1]}})
    Module.Param0WriteEnabled=true
    Module.Param1WriteEnabled=true
    Module.Param2WriteEnabled=true
    Module.Param3WriteEnabled=true
   )
  ]
  rendererClasses [&NiagaraMeshRendererProperties]
  renderers [
   {rendererClass=&NiagaraMeshRendererProperties, rendererIndex=0}
   {rendererClass=&NiagaraMeshRendererProperties, rendererIndex=1}
   {rendererClass=&NiagaraMeshRendererProperties, rendererIndex=2}
  ]
 }
}
summary {
 systemName=NS_JonesyBackpack_RiftOrb
 emitters [
  {
   emitterName=Light
   rendererClasses [
    &NiagaraSpriteRendererProperties
    &NiagaraLightRendererProperties
    &NiagaraComponentRendererProperties
   ]
  }
  {emitterName=Fire, rendererClasses [&NiagaraSpriteRendererProperties]}
  {emitterName=Arcs, rendererClasses [&NiagaraSpriteRendererProperties]}
  {emitterName=PlasmaBall001, rendererClasses [&NiagaraMeshRendererProperties]}
  {emitterName=Embers, rendererClasses [&NiagaraSpriteRendererProperties]}
  {emitterName=Shards, rendererClasses [&NiagaraMeshRendererProperties]}
  {emitterName=ShardsSmall, rendererClasses [&NiagaraMeshRendererProperties]}
 ]
 userVariables [
  User.RibbonIDOffset {type=NiagaraInt32}
  User.Shards_MID {
   type=MaterialInterface
   description=""
   defaultValue {
    struct=NiagaraExt_VariableValue_Object
    value {
     object=&/MI_JonesyBackpack_Shards_Frag_Bit
     objectClass=/Script/Engine.MaterialInterface
    }
   }
  }
  User.Stage {type=NiagaraInt32, defaultValue=3}
 ]
}
